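# Material function for weapon skins that layers a tiled sequin look onto incoming material attributes within a channel-selected mask.
# Two falloff colors blend by Fresnel and a tiled offset pattern, optionally tinted by base color, then scaled by brightness.
# Tiling textures drive roughness and added normal detail, with static switches to keep base values; metalness blends toward base by scalar.
# A static switch bypasses the whole effect.
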MaterialFunction /BRCosmetics/Weapons/WeaponSkins/MaterialFunction/MF_Sequins_ManualEntry
graph {nodes [
  n1 = FunctionInput(Preview=SetMaterialAttributes!(MaterialAttributes)) {bUsePreviewValueAsDefault=true, inputName=inMat, inputType=FunctionInput_MaterialAttributes}
  n2 = GetMaterialAttributes(None=@n1) {attributeGetTypes [
    #1
    #2
    #3
    #4
   ]}
  n3 = GetMaterialAttributes(None=@n1) {attributeGetTypes [
    #3
    #4
    #1
    #2
   ]}
  n4 = TextureObject()
  n5 = ChannelMaskParameter(None=TextureSample(Tex=FunctionInput(Preview=@n4) {bUsePreviewValueAsDefault=true, inputName=Mask, inputType=FunctionInput_Texture2D}).RGBA) {group=MF_Sequin, parameterName=SequinMaskChannel}
  n6 = FunctionInput() {bUsePreviewValueAsDefault=true, inputName=sequinBrightness, inputType=FunctionInput_Scalar, previewValue=<1.5 0 0 1>}
  n7 = Power(Base=Fresnel() {exponent=1})
  n8 = OneMinus(None=@n7)
  n9 = Multiply(A=NamedRerouteUsage(), B=LinearInterpolate(Alpha=Multiply(A=@n8, B=@n7)) {constA=0.7})
  n10 = FunctionInput() {bUsePreviewValueAsDefault=true, inputName=SequinMetalness, inputType=FunctionInput_Scalar, previewValue=<0.9 0 0 1>}
  n11 = NamedRerouteUsage()
  n12 = FunctionInput() {bUsePreviewValueAsDefault=true, inputName=UseBaseMetalness, inputType=FunctionInput_Scalar}
  n13 = NamedRerouteUsage()
  n14 = FunctionInput() {bUsePreviewValueAsDefault=true, inputName=SequinMaxRoughness, inputType=FunctionInput_Scalar, previewValue=<1 0 0 1>}
  n15 = TextureSampleParameter2D(UVs=NamedRerouteUsage()) {
   group=MF_Sequin
   parameterName=Sequin_TilingRoughness
   samplerType=SAMPLERTYPE_LinearGrayscale
   texture=/Game/Global/Textures/Default/Tileable_Patterns/T_SequinTile_roughness
  }
  n16 = LinearInterpolate(A=FunctionInput() {bUsePreviewValueAsDefault=true, inputName=SequinMinRoughness, inputType=FunctionInput_Scalar}, B=@n14, Alpha=@n15.G)
  n17 = NamedRerouteUsage()
  n18 = TextureSampleParameter2D(UVs=NamedRerouteUsage()) {
   group=MF_Sequin
   parameterName=Sequin_TilingNormal
   samplerType=SAMPLERTYPE_Normal
   texture=/Game/Global/Textures/Default/Tileable_Patterns/T_SequinTile_N
  }
  n19 = FunctionInput() {bUsePreviewValueAsDefault=true, inputName=SequinNormalItensity, inputType=FunctionInput_Scalar, previewValue=<0.8 0 0 1>}
  n20 = Multiply(A=@n19, B=Constant3Vector() {constant=<rgba 1 1 0 0>})
  n21 = Add(A=@n17, B=Multiply(A=@n18.RGB, B=@n20)) {constB=0}
  n22 = SetMaterialAttributes!(
   MaterialAttributes=@n2.MaterialAttributes
   BaseColor=LinearInterpolate(A=@n3.BaseColor, B=Multiply(A=@n6, B=@n9), Alpha=@n5.R)
   Metallic=LinearInterpolate(A=@n3.Metallic, B=LinearInterpolate(A=@n10, B=@n11, Alpha=@n12), Alpha=@n5.R)
   Roughness=LinearInterpolate(A=@n3.Roughness, B=StaticSwitchParameter(True=@n13, False=@n16) {group=MF_Sequin, parameterName=UseBaseRoughness}, Alpha=@n5.R)
   Normal=LinearInterpolate(A=@n3.Normal, B=StaticSwitchParameter(True=@n17, False=@n21) {group=MF_Sequin, parameterName=UseBaseNormal}, Alpha=@n5.R)
  ) {attributeSetTypes [
    #3
    #4
    #1
    #2
   ]}
  FunctionOutput(None=StaticSwitchParameter(True=@n1, False=@n22) {parameterName=BypassManualSequin})
  Multiply(A=@n17, B=@n20)
  n23 = FunctionInput() {bUsePreviewValueAsDefault=true, inputName=sequinTile, inputType=FunctionInput_Scalar, previewValue=<32 0 0 1>}
  NamedRerouteDeclaration(None=Multiply(A=TextureCoordinate(), B=@n23)) {name=SequinUVs, nodeColor=<rgba 0 1 0.0859 1>}
  n24 = TextureSampleParameter2D(UVs=NamedRerouteUsage()) {
   group=MF_Sequin
   parameterName=Sequin_TilingOffsetPattern
   samplerType=SAMPLERTYPE_LinearGrayscale
   texture=/Game/Global/Textures/Default/Tileable_Patterns/T_SequinTile
  }
  NamedRerouteDeclaration(None=@n24.R) {name=SequinOffset, nodeColor=<rgba 1 0 0.6562 1>}
  n25 = FunctionInput() {bUsePreviewValueAsDefault=true, inputName=SequinFalloffColor01}
  n26 = FunctionInput() {inputName=SequinFalloffColor02}
  n27 = Multiply(A=LinearInterpolate(Alpha=NamedRerouteUsage()) {constA=0.5, constB=1.5}, B=@n8)
  n28 = NamedRerouteUsage()
  n29 = Multiply(A=@n28, B=@n25)
  n30 = Multiply(A=@n28, B=@n26)
  n31 = FunctionInput() {bUsePreviewValueAsDefault=true, inputName=UseBaseColor, inputType=FunctionInput_Scalar, previewValue=<1 0 0 1>}
  NamedRerouteDeclaration(None=LinearInterpolate(A=LinearInterpolate(A=@n25, B=@n26, Alpha=@n27), B=LinearInterpolate(A=@n29, B=@n30, Alpha=@n27), Alpha=@n31)) {name=SequinColor, nodeColor=<rgba 0.8984 0 1 1>}
  NamedRerouteDeclaration(None=@n2.Roughness) {name=BaseRoughness, nodeColor=<rgba 0.6172 1 0 1>}
  NamedRerouteDeclaration(None=@n2.Normal) {name=BaseNormals, nodeColor=<rgba 0 1 0.1562 1>}
  NamedRerouteDeclaration(None=@n2.BaseColor) {name=BaseColor, nodeColor=<rgba 0 1 0.7188 1>}
  NamedRerouteDeclaration(None=@n2.Metallic) {name=BaseMEtalness, nodeColor=<rgba 0.6406 0 1 1>}
 ]}
