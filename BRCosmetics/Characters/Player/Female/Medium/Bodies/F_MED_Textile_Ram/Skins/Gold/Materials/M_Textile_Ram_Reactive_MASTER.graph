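# Reactive master material for the Gold skin of the Textile Ram female character body, transitioning between two sequin-finished looks.
# Super-level and time parameters sample a curve atlas to animate a moving, growing sphere mask in pre-skinned space, broken up by a fishscale texture.
# The mask blends an alt branch, which restores initial diffuse and specular textures through ReactiveMask, with the main branch.
# Both branches run CharacterSkinFX and manual sequins between shared character PreFX and PostFX functions.
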Material /BRCosmetics/Characters/Player/Female/Medium/Bodies/F_MED_Textile_Ram/Skins/Gold/Materials/M_Textile_Ram_Reactive_MASTER
graph {
 properties {
  bAutomaticallySetUsageInEditor=false
  bUseMaterialAttributes=true
  bUsedWithMeshDeformer=true
  bUsedWithMorphTargets=true
  bUsedWithSkeletalMesh=true
  pixelDepthOffsetMode=PDOM_Legacy
  refractionMethod=RM_IndexOfRefraction
  shadingModel=MSM_FromMaterialExpression
  subsurfaceProfile=&SS_HeroSkin_02
  usedShadingModels=Default Lit | Subsurface | Clear Coat | Subsurface Profile | Cloth | Eye
 }
 nodes [
  Constant() {r=1}
  n1 = MaterialFunctionCall() {materialFunction=&BaseFN_CharacterFunction_PreFX}
  n2 = ScalarParameter() {group=Sequin_Manual, parameterName=useBaseMetalness_Main}
  n3 = TextureObjectParameter() {group=Skin FX, parameterName=SkinFX_Mask}
  n4 = VectorParameter() {group=Sequin_Manual, parameterName=SequinFalloffColor01_Main}
  n5 = VectorParameter() {group=Sequin_Manual, parameterName=SequinFalloffColor02_Main, defaultValue=<rgba 0.8842 0.6282 1 1>}
  n6 = ScalarParameter() {defaultValue=0.8, group=Sequin_Manual, parameterName=useBaseColor_Main}
  n7 = ScalarParameter() {defaultValue=0.8, group=Sequin_Manual, parameterName=sequinMetalness_Main}
  n8 = MaterialFunctionCall(
   PreskinnedPosition=@n1.Pre skinned Local Position
   Gmap=@n1.Pre skinned Local Position
   MaterialAttributes=@n1.Pre skinned Local Position
   AO=@n1.Pre skinned Local Position
  ) {materialFunction=&CharacterSkinFX}
  n9 = ScalarParameter() {defaultValue=0.8, group=Sequin_Manual, parameterName=sequinMaxRoughness_Main}
  n10 = ScalarParameter() {group=Sequin_Manual, parameterName=sequinMinRoughness_Main}
  n11 = ScalarParameter() {defaultValue=0.8, group=Sequin_Manual, parameterName=sequinBrightness_Main}
  n12 = ScalarParameter() {defaultValue=0.8, group=Sequin_Manual, parameterName=sequinTile_Main}
  n13 = ScalarParameter() {defaultValue=0.8, group=Sequin_Manual, parameterName=SequinNormalIntensity_Main}
  n14 = Constant() {r=10}
  n15 = TextureSampleParameter2D() {
   parameterName=FishscaleTex
   samplerType=SAMPLERTYPE_Masks
   texture=/BRCosmetics/Characters/Player/Female/Medium/Bodies/F_MED_Textile_Ram/Textures/uvfix_Body_BaseColor
  }
  n16 = Saturate(None=@n15.R)
  n17 = ScalarParameter() {defaultValue=0.5, group=Reactive, parameterName=Grad_Scale_Negative_Shift}
  n18 = Ceil(None=Add(A=NamedRerouteUsage(), B=Subtract(A=@n16, B=@n17)))
  n19 = ScalarParameter() {defaultValue=1, group=Reactive, parameterName=UseFishscaleMasking}
  n20 = Ceil(None=Multiply(A=NamedRerouteUsage(), B=LinearInterpolate(A=@n14, B=@n18, Alpha=@n19)))
  n21 = ScalarParameter() {group=Sequin_Manual_Alt, parameterName=useBaseMetalness_alt}
  n22 = TextureObjectParameter() {group=Skin FX, parameterName=SkinFX_Mask}
  n23 = VectorParameter() {group=Sequin_Manual_Alt, parameterName=SequinFalloffColor01_alt, defaultValue=<rgba 0.8842 0.6282 1 1>}
  n24 = VectorParameter() {group=Sequin_Manual_Alt, parameterName=SequinFalloffColor02_alt, defaultValue=<rgba 0.6663 0.575 1 1>}
  n25 = ScalarParameter() {defaultValue=1, group=Sequin_Manual_Alt, parameterName=useBaseColor_alt}
  n26 = ScalarParameter() {defaultValue=0.9, group=Sequin_Manual_Alt, parameterName=sequinMetalness_alt}
  n27 = GetMaterialAttributes(None=@n1.Result) {attributeGetTypes [
    #1
    #2
    #3
    #4
   ]}
  n28 = TextureSampleParameter2D() {
   parameterName=InitialDiffuse
   texture=/BRCosmetics/Characters/Player/Female/Medium/Bodies/F_MED_Textile_Ram/Textures/T_F_MED_Textile_Ram_Body_D
  }
  n29 = TextureSampleParameter2D() {
   parameterName=ReactiveMask
   samplerType=SAMPLERTYPE_Masks
   texture=/BRCosmetics/Characters/Player/Female/Medium/Bodies/F_MED_Textile_Ram/Textures/Textile_Ram_TransitionMasks_Body
  }
  n30 = TextureSampleParameter2D() {
   parameterName=InitialSpecular
   samplerType=SAMPLERTYPE_Masks
   texture=/BRCosmetics/Characters/Player/Female/Medium/Bodies/F_MED_Textile_Ram/Textures/T_F_MED_Textile_Ram_Body_S
  }
  n31 = SetMaterialAttributes!(
   MaterialAttributes=@n27.MaterialAttributes
   Base Color=LinearInterpolate(A=@n27.Base Color, B=@n28.RGB, Alpha=@n29.R)
   Specular=LinearInterpolate(A=@n27.Specular, B=@n30.R, Alpha=@n29.R)
   Curvature or Metallic=LinearInterpolate(A=@n27.Curvature or Metallic, B=@n30.G, Alpha=@n29.R)
   Roughness=LinearInterpolate(A=@n27.Roughness, B=@n30.B, Alpha=@n29.R)
  ) {attributeSetTypes [
    #1
    #3
    #2
    #4
   ]}
  n32 = MaterialFunctionCall(
   PreskinnedPosition=@n1.Pre skinned Local Position
   Gmap=@n1.Pre skinned Local Position
   MaterialAttributes=@n31
   AO=@n1.Pre skinned Local Position
  ) {materialFunction=&CharacterSkinFX}
  n33 = ScalarParameter() {defaultValue=0.8, group=Sequin_Manual_Alt, parameterName=sequinMaxRoughness_alt}
  n34 = ScalarParameter() {group=Sequin_Manual_Alt, parameterName=sequinMinRoughness_alt}
  n35 = ScalarParameter() {defaultValue=1.5, group=Sequin_Manual_Alt, parameterName=sequinBrightness_alt}
  n36 = ScalarParameter() {defaultValue=32, group=Sequin_Manual_Alt, parameterName=sequinTile_alt}
  n37 = ScalarParameter() {defaultValue=0.8, group=Sequin_Manual_Alt, parameterName=SequinNormalIntensity_alt}
  n38 = MaterialFunctionCall(
   AO=@n1.AO
   Distance Blend=@n1.AO
   Fresnel=@n1.AO
   Gmap=@n1.AO
   Pre Skinned Local Position=@n1.AO
   Material=BlendMaterialAttributes(
    A=MaterialFunctionCall!(
     UseBaseMetalness=@n21
     Mask=@n22
     SequinFalloffColor01=@n23.RGB
     SequinFalloffColor02=@n24.RGB
     UseBaseColor=@n25
     SequinMetalness=@n26
     inMat=@n32.Result
     SequinMaxRoughness=@n33
     SequinMinRoughness=@n34
     sequinBrightness=@n35
     sequinTile=@n36
     SequinNormalItensity=@n37
    ) {materialFunction=/BRCosmetics/Weapons/WeaponSkins/MaterialFunction/MF_Sequins_ManualEntry}.Result
    B=MaterialFunctionCall(
     UseBaseMetalness=@n2
     Mask=@n3
     SequinFalloffColor01=@n4.RGB
     SequinFalloffColor02=@n5.RGB
     UseBaseColor=@n6
     SequinMetalness=@n7
     inMat=@n8.Result
     SequinMaxRoughness=@n9
     SequinMinRoughness=@n10
     sequinBrightness=@n11
     sequinTile=@n12
     SequinNormalItensity=@n13
    ) {materialFunction=&MF_Sequins_ManualEntry}.Result
    Alpha=Saturate(None=@n20)
   )
  ) {materialFunction=&BaseFN_CharacterFunction_PostFX}
  MaterialFunctionCall(In=@n38.Result) {materialFunction=&CINE_MF_CharPostFX}
  n39 = ScalarParameter() {defaultValue=100, group=SuperLevelling_Mask, parameterName=SL_CurrentLevel, sortPriority=3}
  n40 = Subtract(A=Time(), B=ScalarParameter() {defaultValue=-99999, group=SuperLevelling_Mask, parameterName=SL_GameTimeFromBP, sortPriority=100})
  n41 = Add(A=ScalarParameter() {defaultValue=2, group=SuperLevelling_Mask, parameterName=SL_Animation_Duration, sortPriority=101}) {constB=0}
  n42 = LinearInterpolate(A=Subtract(A=@n39), B=@n39, Alpha=Saturate(None=Divide(A=@n40, B=@n41)))
  n43 = ScalarParameter() {defaultValue=100, group=SuperLevelling_Mask, parameterName=SL_StartAtLevel, sortPriority=3}
  n44 = ScalarParameter() {defaultValue=500, group=SuperLevelling_Mask, parameterName=SL_StopAtLevel, sortPriority=3}
  n45 = Divide(A=Subtract(A=Clamp(None=@n42, Min=@n43, Max=@n44), B=@n43), B=Subtract(A=@n44, B=@n43))
  n46 = Max(A=Ceil(None=ScalarParameter() {defaultValue=1, group=SuperLevelling_Mask, parameterName=SL_QuantizationSteps, sliderMax=100, sliderMin=1}))
  n47 = Floor(None=Multiply(A=@n45, B=@n46))
  n48 = Ceil(None=Subtract(A=@n46))
  n49 = CurveAtlasRowParameter(CurveTime=LinearInterpolate(A=@n45, B=Divide(A=@n47, B=@n46), Alpha=Saturate(None=@n48))) {
   atlas=&AthenaCurveAtlas01
   curve=/Game/Athena/LinearColorCurves/Curve_SuperlevelMask_Summer
   defaultValue=289
   group=SuperLevelling_Mask
   parameterName=SL_SuperLevelMaskCurve
  }
  NamedRerouteDeclaration(None=ChannelMaskParameter(None=@n49) {group=Reactive, parameterName=CurveChannel}.R) {name=AnimationTiming, nodeColor=<rgba 0 1 0.1094 1>}
  n50 = VectorParameter() {group=SuperLevelling_Mask, parameterName=SL_MaskPos_01, defaultValue=<rgba -128 0 64 1>}
  n51 = VectorParameter() {group=SuperLevelling_Mask, parameterName=SL_MaskPos_02, defaultValue=<rgba -128 0 64 1>}
  n52 = VectorParameter() {group=SuperLevelling_Mask, parameterName=SL_MaskPos_03, defaultValue=<rgba -128 0 64 1>}
  n53 = NamedRerouteUsage()
  n54 = Power(Base=@n53, Exp=ScalarParameter() {defaultValue=2, group=SuperLevelling_Mask, parameterName=SL_MaskPos_Exponent})
  n55 = ScalarParameter() {group=SuperLevelling_Mask, parameterName=SL_StartRadius}
  n56 = ScalarParameter() {defaultValue=512, group=SuperLevelling_Mask, parameterName=SL_EndRadius}
  NamedRerouteDeclaration(None=SphereMask(
    A=NamedRerouteUsage()
    B=MaterialFunctionCall(A=@n50.RGB, B=@n51.RGB, C=@n52.RGB, Alpha=@n54) {materialFunction=&Lerp_3Color}.Result
    Radius=LinearInterpolate(A=@n55, B=@n56, Alpha=@n53)
    Hardness=ScalarParameter() {defaultValue=0.5, group=SuperLevelling_Mask, parameterName=SL_CoverMaskHardness}
   )) {name=Spheremask, nodeColor=<rgba 1 0.5156 0 1>}
  NamedRerouteDeclaration(None=@n1.Pre skinned Local Position) {name=Preskinned_Pos, nodeColor=<rgba 0 0.3906 1 1>}
  SetMaterialAttributes!(MaterialAttributes, Base Color) {attributeSetTypes [#1]}
  VertexInterpolator(VS=PreSkinnedPosition())
 ]
}
